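# Animation Blueprint for the Octopus right-arm accessory of the M_MED_Collectable player cosmetic, adapting it to the character's pose.
# It copies the parent mesh pose through the Toy right-arm Control Rig, then applies elbow and upper-arm look-at and additive corrections.
# Each update derives blend values from weapon type, movement, emote and glider states, plus bone rotation deltas.
# Hand-guard and shoulder-band chains are unconnected to the output pose.
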AnimBlueprint /BRCosmetics/Characters/Player/Male/Medium/Bodies/M_MED_Collectable/Meshes/Parts/RightArm/M_MED_Collectable_Octopus_Right_FaceAcc_AnimBP
target_skeleton=&/M_MED_Collectable_Octopus_Right_FaceAcc_Skeleton
graphs [
 AnimGraph {nodes [
   n1 = "Constraint - dyn_arm_r_hand_guard (PPPPP)"!(
    ComponentPose
    Alpha=1
    ConstraintWeights_0=0.1639
    ConstraintWeights_1:float=0.2255
    ConstraintWeights_2:float=0.1689
    ConstraintWeights_3:float=0.2693
    ConstraintWeights_4:float=0.1725
   ) -> (Pose) {
    alphaBoolBlend {blendOption=Linear}
    boneToModify {boneName=dyn_arm_r_hand_guard}
    constraintSetup [
     {targetBone {boneName=dyn_arm_r_hand_r}}
     {targetBone {boneName=dyn_dyn_index_metacarpal_r}}
     {targetBone {boneName=dyn_dyn_index_01_r}}
     {targetBone {boneName=dyn_dyn_pinky_metacarpal_r}}
     {targetBone {boneName=dyn_dyn_ring_metacarpal_r}}
    ]
    constraintWeights [1,1,1,1,1]
   }
   "Transform (Modify) Bone - Bone: dyn_arm_r_hand_guard"(ComponentPose=@n1, Alpha=Get Arm R Hand Rot() -> (Arm R Hand Rot:double), Translation="2,0.5,0", Rotation="15,30,0", Scale="1,1,1") {
    rotationMode=BMM_Additive
    rotationSpace=BCS_BoneSpace
    translationMode=BMM_Additive
    translationSpace=BCS_BoneSpace
    alphaBoolBlend {blendInTime=0.1, blendOption=Linear, blendOutTime=0.1}
    boneToModify {boneName=dyn_arm_r_hand_guard}
    rotation {pitch=15, yaw=30}
    translation=<2 0.5>
   }
   n2 = Get Arm R rot Rifle Additive() -> (Arm R rot Rifle Additive:double)
   n3 = "Constraint - dyn_r_shoulder_band (P)"(Alpha=@n2, ConstraintWeights_0=1) {
    alphaBoolBlend {blendOption=Linear}
    boneToModify {boneName=dyn_r_shoulder_band}
    constraintSetup [{targetBone {boneName=dyn_upperarm_r_AIM}}]
    constraintWeights [1]
   }
   n4 = "Transform (Modify) Bone - Bone: dyn_r_shoulder_band"(ComponentPose=@n3, Alpha=Get Arm R rot Rifle Additive(), Translation="0,0,0", Rotation="-30,-10,0", Scale="1,1,1") {
    rotationMode=BMM_Additive
    rotationSpace=BCS_BoneSpace
    translationMode=BMM_Additive
    translationSpace=BCS_BoneSpace
    alphaBoolBlend {blendInTime=0.1, blendOption=Linear, blendOutTime=0.1}
    boneToModify {boneName=dyn_r_shoulder_band}
    rotation {pitch=-30, yaw=-10}
   }
   "Transform (Modify) Bone - Bone: dyn_r_shoulder_band"!(
    ComponentPose=@n4
    bAlphaBoolEnabled:bool=Get skydiving Arm() -> (skydiving Arm:bool)
    Translation="2.000,0.000,0.500"
    Rotation="30,-15,0"
    Scale="1,1,1"
   ) -> (Pose) {
    alphaInputType=Bool
    rotationMode=BMM_Additive
    rotationSpace=BCS_BoneSpace
    translationMode=BMM_Additive
    translationSpace=BCS_BoneSpace
    alphaBoolBlend {blendInTime=0.1, blendOption=Linear, blendOutTime=0.1}
    boneToModify {boneName=dyn_r_shoulder_band}
    rotation {pitch=30, yaw=-15}
    translation {x=2, z=0.5}
   }
   n5 = Look At - Bone: dyn_upper_arm_r_elbow(ComponentPose=Local To Component(LocalPose=Use cached pose 'Arm setup'()), Alpha=1) {
    alphaBoolBlend {blendOption=Linear}
    boneToModify {boneName=dyn_upper_arm_r_elbow}
    lookAtLocation {x=0}
    lookAtTarget {boneReference {boneName=dyn_upperarm_r_AIM_MSTR}}
    lookAt_Axis {axis=<-1 0>}
   }
   n6 = Get has_hand_glider() -> (has_hand_glider:bool)
   n7 = Get bIsParachuteOpened()
   n8 = "Transform (Modify) Bone - Bone: dyn_upperarm_r_AIM"!(
    ComponentPose=@n5
    bAlphaBoolEnabled:bool=OR Boolean(A=Get bIsSkydiveFloating(), B=AND Boolean(A=@n7, B=@n6))
    Translation="3,0,0.000"
    Rotation="0,0.000,35"
    Scale="1,1,1"
   ) -> (Pose) {
    alphaInputType=Bool
    rotationMode=BMM_Additive
    rotationSpace=BCS_BoneSpace
    translationMode=BMM_Additive
    translationSpace=BCS_BoneSpace
    alphaBoolBlend {blendInTime=0.1, blendOption=Linear, blendOutTime=0.1}
    boneToModify {boneName=dyn_upperarm_r_AIM}
    rotation {roll=35}
    translation {x=3}
   }
   n9 = "Transform (Modify) Bone - Bone: dyn_upperarm_r_AIM"!(ComponentPose=@n8, Alpha:float=@n2, Translation="0,0,0", Rotation="0,15,35", Scale="1,1,1") -> (Pose) {
    rotationMode=BMM_Additive
    rotationSpace=BCS_BoneSpace
    translationSpace=BCS_BoneSpace
    alphaBoolBlend {blendOption=Linear}
    boneToModify {boneName=dyn_upperarm_r_AIM}
    rotation {roll=35, yaw=15}
   }
   n10 = Look At - Bone: dyn_upperarm_r_AIM(ComponentPose=@n9, Alpha=Get Parachute Arm() -> (Parachute Arm:double)) {
    alphaBoolBlend {blendOption=Linear}
    alphaScaleBiasClamp {bInterpResult=true}
    boneToModify {boneName=dyn_upperarm_r_AIM}
    lookAtLocation {x=0}
    lookAtTarget {boneReference {boneName=dyn_upper_arm_r_elbow_MSTR}}
    lookAt_Axis {axis=<1 0>}
   }
   Output Pose(Result=Component To Local(ComponentPose=@n10)) {layerGroup=None}
   n11 = Control Rig!(
    Source=Copy Pose From Mesh() {bUseAttachedParent=true}
    Alpha
    Rifle:double=Get Rifle Float() -> (Rifle Float:double)
    Arm Twists:double=1
    Rifle Arm Up:double=Get Arm R rot Rifle Additive()
    Emote And Pickaxe:double=Get Emote and Pickaxe() -> (Emote and Pickaxe:double)
    Parachute Arm:double=Get Parachute Arm()
   ) -> (Pose) {
    alphaBoolBlend {blendOption=Linear}
    controlRigAssetReference {blueprintRigClass=/Game/Characters/Player/Male/Medium/Bodies/M_MED_Collectable/Meshes/Parts/RightArm/M_MED_Collectable_Toy_Right_FaceAcc_CtrlRig.M_MED_Collectable_Toy_Right_FaceAcc_CtrlRig_C}
   }
   Save cached pose 'Arm setup'(Pose=@n11) {cacheName=Arm setup}
   Get Arm Twist Floats() -> (Arm Twist Floats:double)
  ]}
 EventGraph {nodes [
   Event BlueprintUpdateAnimation()
   > n1 = Sequence!(execute) -> (then_0, then_1, then_2:Exec, then_3:Exec, then_4:Exec, then_5:Exec, then_6:Exec, then_7:Exec, then_8:Exec, then_9:Exec)
   n2 = Cast To FortPlayerAnimInstance(execute=@n1.then_1, Object=GetAnimInstance(self=GetOwnerSkeletalMeshComponent()))
   > Set Rifle(Rifle:bool=false) -> (Output_Get:bool)
   > Set Pistol(Pistol:bool=false) -> (Output_Get:bool)
   > Set Melee(Melee:bool=false) -> (Output_Get:bool)
   > n3 = Switch on EFortWeaponCoreAnimation!(execute, Selection:EFortWeaponCoreAnimation Enum=Get WeaponCoreAnim(self=@n2)) -> (
    Melee:Exec
    Pistol:Exec
    Shotgun:Exec
    PaperBlueprint:Exec
    Rifle:Exec
    MeleeOneHand:Exec
    MachinePistol:Exec
    RocketLauncher:Exec
    GrenadeLauncher:Exec
    GoingCommando:Exec
    AssaultRifle:Exec
    TacticalShotgun:Exec
    SniperRifle:Exec
    TrapPlacement:Exec
    ShoulderLauncher:Exec
    AbilityDecoTool:Exec
    Crossbow:Exec
    C4:Exec
    RemoteControl:Exec
    DualWield:Exec
    AR_BullPup:Exec
    AR_ForwardGrip:Exec
    MedPackPaddles:Exec
    SMG_P90:Exec
    AR_DrumGun:Exec
    Consumable_Small:Exec
    Consumable_Large:Exec
    Balloon:Exec
    MountedTurret:Exec
    CreativeTool:Exec
    ExplosiveBow:Exec
    AshtonIndigo:Exec
    AshtonChicago:Exec
    MeleeDualWield:Exec
    Unarmed:Exec
   )
   n4 = Branch(
    execute=[@n3.AssaultRifle, @n3.Crossbow, @n3.GrenadeLauncher, @n3.Rifle, @n3.RocketLauncher, @n3.Shotgun, @n3.SniperRifle, @n3.TacticalShotgun]
    Condition=Get bIsSkyDiving()
   )
   Set Rifle(execute=@n4.else, Rifle=true)
   n5 = Branch(execute=[@n3.MachinePistol, @n3.Pistol], Condition=Get bIsSkyDiving())
   Set Pistol(execute=@n5.else, Pistol=true)
   n6 = Branch(execute=@n3.Melee, Condition=Get bIsSkyDiving())
   Set Melee(execute=@n6.else, Melee=true)
   Set Rifle Float(execute=@n1.then_2, Rifle Float:double=SelectFloat(A=-0.15, B=0.15, bPickA=Get Pistol() -> (Pistol:bool))) -> (Output_Get:double)
   > n7 = Branch(Condition=OR Boolean!(A=Get bIsFalling(), B=Get bIsJumping(), C:bool=Get bIsSkyDiving(), D:bool=Get bIsPlayingEmote()) -> (ReturnValue))
   > Set Arm R rot(Arm R rot:double=0) -> (Output_Get:double)
   Set Arm R rot(
    execute=@n7.else
    Arm R rot=GetAlphaFromReferenceTransformDelta(SkeletalMeshComponent=Get Skel_Mesh(), BoneName=upperarm_r, BaseName=spine_05, InputRange=(X=60,Y=100), TwistAxis="1,0,0")
   )
   n8 = Branch(execute=@n1.then_3, Condition=Get Rifle() -> (Rifle:bool))
   > n9 = Branch(Condition=OR Boolean(A=Get bIsFalling(), B=Get bIsJumping(), C=Get bIsSkyDiving(), D=Get bIsPlayingEmote()))
   > Set Arm R rot Rifle Additive(Arm R rot Rifle Additive:double=0) -> (Output_Get:double)
   > Set Arm R Hand Rot(Arm R Hand Rot:double=0) -> (Output_Get:double)
   Set Arm R rot Rifle Additive(
    execute=@n9.else
    Arm R rot Rifle Additive=GetAlphaFromReferenceTransformDelta(SkeletalMeshComponent=Get Skel_Mesh(), BoneName=upperarm_l, BaseName=spine_05, InputRange=(X=20,Y=100), TwistAxis="1,0,0")
   )
   > Set Arm R rot Rifle Clav Additive(Arm R rot Rifle Clav Additive:double=GetAlphaFromReferenceTransformDelta(
     SkeletalMeshComponent=Get Skel_Mesh()
     BoneName=clavicle_l
     BaseName=spine_05
     DeltaComponentType=EulerY
     InputRange=(X=30,Y=0)
     TwistAxis="1,0,0"
    )) -> (Output_Get:double)
   > Set Arm R Hand Rot(Arm R Hand Rot=GetAlphaFromReferenceTransformDelta(
     SkeletalMeshComponent=Get Skel_Mesh()
     BoneName=hand_r
     BaseName=lowerarm_r
     DeltaComponentType=EulerY
     InputRange=(X=30,Y=60)
     TwistAxis="1,0,0"
    ))
   Set Arm R rot Rifle Additive(execute=@n8.else, Arm R rot Rifle Additive=0)
   > Set Arm R Hand Rot(Arm R Hand Rot=0)
   n10 = Branch(execute=@n1.then_4, Condition=OR Boolean!(A=Get bIsSkydiveFloating(), B=Get bIsParachuteOpened()) -> (ReturnValue))
   > Set skydiving Arm(skydiving Arm:bool=true) -> (Output_Get:bool)
   Set skydiving Arm(execute=@n10.else, skydiving Arm=false)
   n11 = Branch(
    execute=@n1.then_5
    Condition=OR Boolean!(A=Get bIsPlayingEmote(), B=Get bIsPlayingMeleeAnim(), C:bool=Get Melee() -> (Melee:bool)) -> (ReturnValue)
   )
   > Set Emote and Pickaxe(Emote and Pickaxe:double=1) -> (Output_Get:double)
   Set Emote and Pickaxe(execute=@n11.else, Emote and Pickaxe=0)
   Branch(execute=@n1.then_6, Condition=Get bIsSkyDiving())
   > Set Rifle Float(Rifle Float=0.2)
   n12 = Branch(execute=@n1.then_7, Condition=Get bIsParachuteOpened())
   > Set Parachute Arm(Parachute Arm:double=1) -> (Output_Get:double)
   Set Parachute Arm(execute=@n12.else, Parachute Arm=0)
   n13 = Branch(execute=@n1.then_8, Condition=Get Pistol())
   > n14 = Branch(Condition=OR Boolean!(A=Get bIsFalling(), B=Get bIsJumping(), C=Get bIsSkyDiving(), D:bool=Get bIsPlayingEmote()) -> (ReturnValue))
   > Set Arm R rot Pistol Additive(Arm R rot Pistol Additive:double=0) -> (Output_Get:double)
   Set Arm R rot Pistol Additive(
    execute=@n14.else
    Arm R rot Pistol Additive=GetAlphaFromReferenceTransformDelta(SkeletalMeshComponent=Get Skel_Mesh(), BoneName=upperarm_r, BaseName=spine_05, InputRange=(X=80,Y=130), TwistAxis="1,0,0")
   )
   Set Arm R rot Pistol Additive(execute=@n13.else, Arm R rot Pistol Additive=0)
   n15 = Get fort_player() -> (fort_player:Fort Player Anim Instance Object Reference)
   Switch on EGliderType(
    execute=@n1.then_9
    Selection=Break Fort Anim Input Player Glider Anim Asset!(FortAnimInput_PlayerGliderAnimAsset:"Fort Anim Input Player Glider Anim Asset Structure (by ref)"=Get CurrentGliderAnimSet!(self:Fort Player Anim Instance Object Reference=@n15) -> (CurrentGliderAnimSet:Fort Anim Input Player Glider Anim Asset Structure)) -> (
     Default_Base_BS:Blend Space Object Reference
     Default_BodyAdditive_MaleMedium_BS:Blend Space Object Reference
     Default_BodyAdditive_MaleLarge_BS:Blend Space Object Reference
     Default_BodyAdditive_FemaleSmall_BS:Blend Space Object Reference
     Default_BodyAdditive_FemaleMedium_BS:Blend Space Object Reference
     Default_BodyAdditive_FemaleLarge_BS:Blend Space Object Reference
     Default_TurnAdditive_MaleMedium_BS:Blend Space Object Reference
     Default_TurnAdditive_MaleLarge_BS:Blend Space Object Reference
     Default_TurnAdditive_FemaleSmall_BS:Blend Space Object Reference
     Default_TurnAdditive_FemaleMedium_BS:Blend Space Object Reference
     Default_TurnAdditive_FemaleLarge_BS:Blend Space Object Reference
     Into_Base_BS:Blend Space Object Reference
     Into_Base_MaleMedium_BS:Blend Space Object Reference
     Into_BodyAdditive_MaleMedium_BS:Blend Space Object Reference
     Into_BodyAdditive_MaleLarge_BS:Blend Space Object Reference
     Into_BodyAdditive_FemaleSmall_BS:Blend Space Object Reference
     Into_BodyAdditive_FemaleMedium_BS:Blend Space Object Reference
     Into_BodyAdditive_FemaleLarge_BS:Blend Space Object Reference
     Into_TurnAdditive_MaleMedium_BS:Blend Space Object Reference
     Into_TurnAdditive_MaleLarge_BS:Blend Space Object Reference
     Into_TurnAdditive_FemaleSmall_BS:Blend Space Object Reference
     Into_TurnAdditive_FemaleMedium_BS:Blend Space Object Reference
     Into_TurnAdditive_FemaleLarge_BS:Blend Space Object Reference
     Lean_MaleMedium_BS:Blend Space Object Reference
     Lean_MaleLarge_BS:Blend Space Object Reference
     Lean_FemaleSmall_BS:Blend Space Object Reference
     Lean_FemaleMedium_BS:Blend Space Object Reference
     Lean_FemaleLarge_BS:Blend Space Object Reference
     LeanAdditive_Center_MaleMedium_Pose:Anim Sequence Object Reference
     LeanAdditive_Center_MaleLarge_Pose:Anim Sequence Object Reference
     LeanAdditive_Center_FemaleSmall_Pose:Anim Sequence Object Reference
     LeanAdditive_Center_FemaleMedium_Pose:Anim Sequence Object Reference
     LeanAdditive_Center_FemaleLarge_Pose:Anim Sequence Object Reference
     LeanAdditive_Into_BS:Blend Space Object Reference
     LeanAdditive_ForwardInto_Anim:Anim Sequence Object Reference
     LeanAdditive_ForwardInto_FromDeploy_Anim:Anim Sequence Object Reference
     LeanAdditive_BackInto_Anim:Anim Sequence Object Reference
     LeanAdditive_LeftInto_Anim:Anim Sequence Object Reference
     LeanAdditive_RightInto_Anim:Anim Sequence Object Reference
     LeanAdditive_ForwardOut_Anim:Anim Sequence Object Reference
     LeanAdditive_BackOut_Anim:Anim Sequence Object Reference
     LeanAdditive_LeftOut_Anim:Anim Sequence Object Reference
     LeanAdditive_RightOut_Anim:Anim Sequence Object Reference
     ToGlide_BS:Blend Space Object Reference
     ToGlide_Lean_BS:Blend Space Object Reference
     ToDive_BS:Blend Space Object Reference
     ToDive_Lean_BS:Blend Space Object Reference
     Dive_WeaponR_Additive_BS:Blend Space Object Reference
     Glide_WeaponR_Additive_Anim:Anim Sequence Object Reference
     GenericAdditive_Male_BS:Blend Space Object Reference
     GenericAdditive_Female_BS:Blend Space Object Reference
     RootModPitchMin:float=~
     RootModPitchMax:float=~
     RootModYOffsetMin:float=~
     RootModYOffsetMax:float=~
     PelvisModVertical:Pelvis Mod Vertical Input Spring Structure
     PelvisModLateral:Pelvis Mod Lateral Input Spring Structure
     PelvisModBankAngle:Pelvis Mod Bank Angle Input Spring Structure
     ThighModLegAngle:Thigh Mod Leg Angle Input Structure
     ThighModLegPitch:Thigh Mod Leg Pitch Spring Structure
     ThighModLegBank:Thigh Mod Leg Bank Spring Structure
     ThighModLeftLegDrag:Thigh Mod Left Leg Drag Structure
     PlayerGliderType:EGliderType Enum
     bEnableSpringMods:bool=~
     bAllowPlayerDeployRootMod:bool=~
     bUseSurfStyle:bool=~
     bUseLeanPlayRateAdjust:bool=~
     bUseForwardAs_Y_BS:bool=~
     bUseDeployCurveForIntoPlayRateAdjust:bool=~
     LeanForwardValue:float=~
     IdleValue:float=~
     LeanBackValue:float=~
     IdleInterpToSpeed:float=~
     LeanForwardInterpToSpeed:float=~
     LeanBackInterpToSpeed:float=~
    ) {is=title K2Node_BreakStruct}.PlayerGliderType
   )
   > Set has_hand_glider(has_hand_glider:bool=true) -> (Output_Get:bool)
   Event BlueprintInitializeAnimation()
   n16 = GetOwnerSkeletalMeshComponent()
   > Set Skel_Mesh(Skel_Mesh=@n16)
   > Set Body_Mesh(Body_Mesh=@n16)
   > Set fort_player(fort_player:Fort Player Anim Instance Object Reference=GetOwnerFortPlayerAnimInstance!(self:Custom Character Part Anim Instance Object Reference) -> (ReturnValue:Fort Player Anim Instance Object Reference)) -> (Output_Get:Fort Player Anim Instance Object Reference)
   TryGetPawnOwner()
   Get Rifle()
   Set Rifle Float(Rifle Float=SelectFloat(A=-0.15, B=0.15, bPickA=false))
  ]}
]
variables [
 Arm R Hand Rot {type=number}
 Arm R rot {type=number}
 Arm R rot Down {type=number}
 Arm R rot Pistol Additive {type=number}
 Arm R rot Rifle Additive {type=number}
 Arm R rot Rifle Clav Additive {type=number}
 Arm Twist Floats {type=number}
 Base Arms {type=boolean}
 Body_Mesh {type=SkeletalMeshComponent}
 Emote and Pickaxe {type=number}
 fort_player {type=FortPlayerAnimInstance, default=null}
 has_hand_glider {type=boolean}
 Melee {type=boolean}
 Parachute Arm {type=number}
 Pistol {type=boolean}
 Razor Arms {type=boolean}
 Rifle {type=boolean}
 Rifle Float {type=number}
 Robot Arms {type=boolean}
 Skel_Mesh {type=SkeletalMeshComponent}
 skydiving Arm {type=boolean}
 Toy Arms {type=boolean}
 Tube ARms {type=boolean}
]
